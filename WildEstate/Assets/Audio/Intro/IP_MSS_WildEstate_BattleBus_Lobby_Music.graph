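# Stereo MetaSound source for the Battle Bus lobby music in the WildEstate intro.
# On play, it picks a random time up to the length of its wave asset variable and starts the stereo wave player there, so playback begins at a different point each time.
# The source is flagged looping, uses the Athena bus lobby music attenuation and the gadgets sound class, and overrides concurrency to one instance.
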MetaSoundSource /WildEstate/Assets/Audio/Intro/IP_MSS_WildEstate_BattleBus_Lobby_Music
inputs [
 UE.Attenuation.Distance {type=Float}
 UE.Source.OnPlay {type=Trigger}
 UE.Spatialization.Azimuth {type=Float}
 UE.Spatialization.Elevation {type=Float}
]
interfaces [UE.Attenuation:1.0, UE.OutputFormat.Stereo:1.0, UE.Source:1.1, UE.Spatialization:1.0]
outputs [
 "UE.OutputFormat.Stereo.Audio:0" {type=Audio}
 "UE.OutputFormat.Stereo.Audio:1" {type=Audio}
]
pages [Default {
  input_defaults {UE.Attenuation.Distance=0, UE.Source.OnPlay=null, UE.Spatialization.Azimuth=0, UE.Spatialization.Elevation=0}
  nodes [
   n1 = VariableAccessor.WaveAsset(Variable=InitVariable.WaveAsset().Variable)
   GetWaveDuration_0 = "UE.Get Wave Duration: "(Wave=@n1.Value)
   n2 = UE.RandomTime(Next=^UE.Source.OnPlay, Max=@GetWaveDuration_0.Duration)
   WavePlayer_0 = UE.Wave Player:Stereo(Play=@n2.On Next, Start Time=@n2.Value, Wave Asset=UE.Reroute(Value=@n1.Value).Value) {also {defaults {Loop="<MetaSoundBuilderResult.SUCCEEDED: 0>"}, version [1,1]}}
  ]
  links [{dst="@UE.OutputFormat.Stereo.Audio:0", src=WavePlayer_0.Out Left}, {dst="@UE.OutputFormat.Stereo.Audio:1", src=WavePlayer_0.Out Right}]
 }]
properties {
 assetImportData=&self:AssetImportData
 attenuationSettings=/Game/Sounds/Attenuation/Fort_Bus_Lobby_Music_Athena_Att
 bLooping=true
 bOverrideConcurrency=true
 duration=10000
 numChannels=2
 outputFormat=Stereo
 sampleRate=48000
 soundClassObject=&Fort_Gadgets
 assetUserData [&self:AudioAssetUserData_0]
 concurrencyOverrides {maxCount=1}
}
