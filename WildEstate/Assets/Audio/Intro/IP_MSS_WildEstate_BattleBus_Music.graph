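# Stereo MetaSound Source for the Wild Estate Battle Bus intro music, handing a transition stinger over to the bus music loop.
# On play, the unattenuated stinger starts at once and the looping track starts after a delay.
# A delayed trigger drives an interpolated crossfade from stinger to loop.
# Only the loop passes through distance-driven spatialization, per-channel air absorption and linear volume falloff, then a gain boost.
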MetaSoundSource /WildEstate/Assets/Audio/Intro/IP_MSS_WildEstate_BattleBus_Music
inputs [
 UE.Attenuation.Distance {type=Float}
 UE.Source.OnPlay {type=Trigger}
 UE.Spatialization.Azimuth {type=Float}
 UE.Spatialization.Elevation {type=Float}
]
interfaces [UE.Attenuation:1.0, UE.OutputFormat.Stereo:1.0, UE.Source:1.1, UE.Spatialization:1.0]
outputs [
 "UE.OutputFormat.Stereo.Audio:0" {type=Audio}
 "UE.OutputFormat.Stereo.Audio:1" {type=Audio}
]
pages [Default {
  input_defaults {UE.Attenuation.Distance=1649.9948, UE.Source.OnPlay=null, UE.Spatialization.Azimuth=0, UE.Spatialization.Elevation=0}
  nodes [
   n1 = InitVariable.Trigger()
   WavePlayer_1 = UE.Wave Player:Stereo(Play=^UE.Source.OnPlay) {also {defaults {Wave Asset=&/MX_S38_IP_StarterIsland_to_BattleBus_Transition_STNG}, version [1,1]}}
   TriggerDelay_0 = UE.Trigger Delay(In=@WavePlayer_1.On Play) {also {defaults {Delay Time=3.8}}}
   TriggerRouteTriggerRouteFloat2_0 = "TriggerRoute.Trigger Route (Float, 2):Float"(
    Set 0=^UE.Source.OnPlay
    Set 1=VariableAccessor.Trigger(Variable=VariableMutator.Trigger(Variable=@n1.Variable, Value=@TriggerDelay_0.Out).Variable).Value
   ) {also {defaults {Value 1=1}}}
   n2 = UE.InterpTo:Audio(Target=@TriggerRouteTriggerRouteFloat2_0.Value) {also {defaults {Interp Time=2}}}
   DecibelstoLinearGain_0 = UE.Decibels to Linear Gain:Float()
   DecibelstoLinearGain_1 = UE.Decibels to Linear Gain:Float() {also {defaults {Decibels=3}}}
   TriggerDelay_1 = UE.Trigger Delay(In=^UE.Source.OnPlay) {also {defaults {Delay Time=4.5}}}
   WavePlayer_0 = UE.Wave Player:Stereo(Play=@TriggerDelay_1.Out) {also {defaults {Loop="<MetaSoundBuilderResult.SUCCEEDED: 0>", Wave Asset=&/MX_S38_IP_BattleBus_LP}, version [1,1]}}
   n3 = /FortMetasound/Attenuation/MSP_Attenuation_Spatialization(Distance=^UE.Attenuation.Distance, Azimuth=^UE.Spatialization.Azimuth, Stereo.InL=@WavePlayer_0.Out Left, Stereo.InR=@WavePlayer_0.Out Right) {also {defaults {NonSpatializedRadiusEnd=1500, StereoSpread=10}}}
   n4 = /FortMetasound/Attenuation/MSP_Attenuation_AirAbsorption(Distance=^UE.Attenuation.Distance, AudioIn=@n3.Right) {also {defaults {LowPassCutoffFrequencyMax=3000, MaxDistanceRange=16000, MinDistanceRange=5000}}}
   n5 = /FortMetasound/Attenuation/MSP_Attenuation_Volume_Linear(Distance=^UE.Attenuation.Distance) {also {defaults {FalloffDistance=20000, InnerRadius=1000}}}
   CrossfadeTriggerRouteAudio2_0 = "Crossfade.Trigger Route (Audio, 2):Audio"(
    Crossfade Value=@n2.Value
    In 0=UE.Multiply:Audio by Float(AdditionalOperands=@DecibelstoLinearGain_0.Linear Gain, PrimaryOperand=@WavePlayer_1.Out Right).Out
    In 1=UE.Multiply:Audio by Float(
     AdditionalOperands=@DecibelstoLinearGain_1.Linear Gain
     PrimaryOperand=UE.Multiply:Audio by Float(PrimaryOperand=@n4.AudioOut, AdditionalOperands=@n5.VolumeMultiplier).Out
    ).Out
   )
   n6 = &MSP_Attenuation_AirAbsorption(Distance=^UE.Attenuation.Distance, AudioIn=@n3.Left) {also {defaults {LowPassCutoffFrequencyMax=3000, MaxDistanceRange=16000, MinDistanceRange=5000}}}
   CrossfadeTriggerRouteAudio2_1 = "Crossfade.Trigger Route (Audio, 2):Audio"(
    Crossfade Value=@n2.Value
    In 0=UE.Multiply:Audio by Float(AdditionalOperands=@DecibelstoLinearGain_0.Linear Gain, PrimaryOperand=@WavePlayer_1.Out Left).Out
    In 1=UE.Multiply:Audio by Float(
     AdditionalOperands=@DecibelstoLinearGain_1.Linear Gain
     PrimaryOperand=UE.Multiply:Audio by Float(PrimaryOperand=@n6.AudioOut, AdditionalOperands=@n5.VolumeMultiplier).Out
    ).Out
   )
  ]
  links [
   {dst="@UE.OutputFormat.Stereo.Audio:1", src=CrossfadeTriggerRouteAudio2_0.Out}
   {dst="@UE.OutputFormat.Stereo.Audio:0", src=CrossfadeTriggerRouteAudio2_1.Out}
  ]
 }]
properties {
 assetImportData=&self:AssetImportData
 attenuationSettings=&/ATT_WildEstate_BattleBus
 bLooping=true
 bOverrideConcurrency=true
 duration=10000
 maxDistance=20000
 numChannels=2
 outputFormat=Stereo
 sampleRate=48000
 soundClassObject=&Fort_Gadgets
 volume=1.4
 assetUserData [&self:AudioAssetUserData_0]
 concurrencyOverrides {maxCount=1}
}
